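# Idle Niagara effect for the EssayViewMyth pickaxe, with socket-attached glows and orange lightning that scale with a system reactivity intensity.
# Capsule_Glow feeds a socket-rotated direction into its material, while Glows places three differently oriented, partitioned sprites.
# Lightning_Splines spawns short-lived meshes at a reactivity-driven rate, varying shape index, mirroring and inversion through particle and material parameters.
# All three emitters tint by time of day.
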NiagaraSystem /BRCosmetics/Weapons/FORT_Melee/Pickaxe_EssayViewMyth/FX/NS_Pickaxe_EssayViewMyth_Idle
dependencies {
 usedDataInterfaces [
  &NiagaraDataInterfaceMeshRendererInfo
  &NiagaraDataInterfaceCurve
  &NiagaraDataInterfaceArrayFloat
  &NiagaraDataInterfaceColorCurve
  &NiagaraDataInterfaceSocketReader
  &NiagaraDataInterfaceArrayInt32
 ]
 usedDynamicInputs [
  &Add_Float
  &Multiply_FloatByInt
  &SpawnGroup
  &SimulationPosition
  &Add_Integer
  &Multiply_Int
  &RandomRangeInt
  &Subtract_Float
 ]
 usedModules [
  &self:SystemSpawnScript.SystemScriptSource.SystemScriptGraph.NiagaraNodeAssignment_0.SetVariables_#1
  &self:SystemSpawnScript.SystemScriptSource.SystemScriptGraph.NiagaraNodeAssignment_1.SetVariables_#2
  &self:SystemSpawnScript.SystemScriptSource.SystemScriptGraph.NiagaraNodeAssignment_2.SetVariables_#3
  &self:SystemSpawnScript.SystemScriptSource.SystemScriptGraph.NiagaraNodeAssignment_4.SetVariables_#4
  &SystemState
  &self:SystemSpawnScript.SystemScriptSource.SystemScriptGraph.NiagaraNodeAssignment_3.SetVariables_#5
 ]
}
emitters {
 Capsule_Glow {
  emitterSpawnScript [/SetVariables(Emitter.CapsuleGlowDirection:Vector3f=<1 0 -0.15>) {isSetParameters=true}]
  emitterUpdateScript [
   &EmitterState(
    Inactive Response=::NewEnumerator1
    Life Cycle Mode=::NewEnumerator1
    Loop Behavior=::NewEnumerator1
    Loop Duration Mode=::NewEnumerator1
    UseLoopDelay=false
   )
   &SpawnBurst_Instantaneous(Use Spawn Probability=false, UseLoopCountLimit=false)
  ]
  particleSpawnScript [
   &InitializeParticle(
    Color=$System.Glow Color
    Lifetime=5
    Position Offset=$System.Position Offset
    Position Offset Coordinate Space=::Local
    Sprite Size Mode=::NewEnumerator3
    Uniform Sprite Size=50
    UsePositionOffset=true
    Write Material Random=true
   )
   &CameraOffset(Camera Offset Amount=5)
  ]
  particleUpdateScript [
   &ParticleState(Kill Particles When Lifetime Has Elapsed=false, Loop Particles Lifetime=true)
   &SocketLocation(
    SamplePosOffset=true
    Sampled Position Offset=<5 0 -4>
    Socket Reader=$System.Socket Reader
    Socket Sampling Mode=::NewEnumerator3
    VelocitySpeedLimit=false
   )
   &ScaleColor(
    Scale Alpha=&Lerp_Float(A=0, B=0.25, Alpha=$System.ReactivityIntensity)
    Scale RGB=&Multiply_VectorByFloat(
     Vector=&MakeVectorFromLinearColorRGB(Color=&FNTimeOfDayColor(Night Color=<rgba 0.56 0.6488 0.8 1>, Morning Color=<rgba 0.6 0.6262 0.8 1>, Day Color=<rgba 1 1 1 1>, Evening Color=<rgba 0.6 0.6 0.6 1>))
     Float=&Lerp_Float(A=3, B=6, Alpha=$System.ReactivityIntensity)
    )
    ScaleA=true
    ScaleRGB=true
   ) {module=ScaleColor001}
   &DynamicMaterialParameters(
    Index 0 Param 1=&MakeFloatFromVector(
     Vector=&RotateVectorByQuaternion(Vector To Rotate=$Emitter.CapsuleGlowDirection, Quaternion=$Output.SocketLocation.SampledRotation)
     Channel=::NewEnumerator0
    )
    Index 0 Param 2=&MakeFloatFromVector(
     Vector=&RotateVectorByQuaternion(Vector To Rotate=$Emitter.CapsuleGlowDirection, Quaternion=$Output.SocketLocation.SampledRotation)
     Channel=::NewEnumerator1
    )
    Index 0 Param 3=&MakeFloatFromVector(
     Vector=&RotateVectorByQuaternion(Vector To Rotate=$Emitter.CapsuleGlowDirection, Quaternion=$Output.SocketLocation.SampledRotation)
     Channel=::NewEnumerator2
    )
    Index 0 Param 4=50
    Module.Param0WriteEnabled=true
    Module.Param1WriteEnabled=true
    Module.Param2WriteEnabled=true
    Module.Param3WriteEnabled=true
   ) {module=DynamicMaterialParameters001}
  ]
  rendererClasses [&NiagaraSpriteRendererProperties]
  renderers [{rendererClass=&NiagaraSpriteRendererProperties}]
 }
 Glows {
  emitterUpdateScript [
   &EmitterState(
    Inactive Response=::NewEnumerator1
    Life Cycle Mode=::NewEnumerator1
    Loop Behavior=::NewEnumerator1
    Loop Duration Mode=::NewEnumerator1
    UseLoopDelay=false
   )
   &SpawnBurst_Instantaneous(Spawn Count=3, Use Spawn Probability=false, UseLoopCountLimit=false)
  ]
  particleSpawnScript [
   &PartitionParticles(Number Of Partitions=3)
   &InitializeParticle(
    Color=$System.Glow Color
    Position Offset=$System.Position Offset
    Sprite Size=&MergeHelper_Vector2Switch(0=<30 30>, 1=<110 110>, 2=<100 100>, 3=?, 4=?, Index=$Particles.PartitionParticles.Partition, Num=::NewEnumerator1)
    Sprite Size Mode=::NewEnumerator4
    UsePositionOffset=true
    Write Material Random=true
   )
   &SpriteFacingAndAlignment(
    Sprite Alignment=&MergeHelper_VectorSwitch(0=<0 0 1>, 1=<0 0 1>, 2=<-1 0 0>, 3=?, Index=$Particles.PartitionParticles.Partition, Num=::NewEnumerator1)
    Sprite Facing=&MergeHelper_VectorSwitch(0=<0 1 0>, 1=<1 0 0>, 2=<0 0.17 1>, 3=?, Index=$Particles.PartitionParticles.Partition, Num=::NewEnumerator1)
    WriteAlignment=true
    WriteFacing=true
   )
   &CameraOffset(Camera Offset Amount=&MergeHelper_FloatSwitch(0=0, 1=-20, 2=-5, 3=?, Index=$Particles.PartitionParticles.Partition, Num=::NewEnumerator1))
   /SetVariables(Particles.VisibilityTag=$Particles.PartitionParticles.Partition) {isSetParameters=true}
  ]
  particleUpdateScript [
   &ParticleState(Kill Particles When Lifetime Has Elapsed=false, Loop Particles Lifetime=true)
   &SocketLocation(
    SamplePosOffset=true
    Sampled Position Offset=&MergeHelper_VectorSwitch(0=<-10 0 -2>, 1=<5 0 -1>, 2=<5 0 -3>, 3=?, Index=$Particles.PartitionParticles.Partition, Num=::NewEnumerator1)
    Socket Reader=$System.Socket Reader
    Socket Sampling Mode=::NewEnumerator3
    VelocitySpeedLimit=false
   )
   &ScaleColor(
    Scale Alpha=&FloatFromCurve(
     FloatCurve=DataInterface() {propertyValues {Curve {keys [<at 0 0.6 RCIM_Cubic>,<at 0.5 1 RCIM_Cubic>,<at 1 0.6 RCIM_Cubic>]}}}
     Scale Curve=&Lerp_Float(A=0, B=0.25, Alpha=$System.ReactivityIntensity)
    )
    Scale RGB=&Multiply_VectorByFloat(
     Vector=&MakeVectorFromLinearColorRGB(Color=&FNTimeOfDayColor(Night Color=<rgba 0.56 0.6488 0.8 1>, Morning Color=<rgba 0.6 0.6262 0.8 1>, Day Color=<rgba 1 1 1 1>, Evening Color=<rgba 0.6 0.6 0.6 1>))
     Float=&Lerp_Float(A=0.5, B=1.5, Alpha=$System.ReactivityIntensity)
    )
    ScaleA=true
    ScaleRGB=true
   )
  ]
  rendererClasses [&NiagaraSpriteRendererProperties]
  renderers [
   {rendererClass=&NiagaraSpriteRendererProperties}
   {rendererClass=&NiagaraSpriteRendererProperties, rendererIndex=1}
   {rendererClass=&NiagaraSpriteRendererProperties, rendererIndex=2}
  ]
 }
 Lightning_Splines {
  emitterUpdateScript [
   &EmitterState(
    Inactive Response=::NewEnumerator1
    Life Cycle Mode=::NewEnumerator1
    Loop Behavior=::NewEnumerator1
    Loop Duration Mode=::NewEnumerator1
    UseLoopDelay=false
   )
   &SpawnRate(
    Spawn Probability=0.8
    SpawnRate=&FloatFromCurve(
     FloatCurve=DataInterface() {propertyValues {Curve {
        defaultValue=1
        keys [
         {interpMode=RCIM_Cubic, tangentMode=RCTM_User, time=0, value=0}
         {arriveTangent=2.2852, interpMode=RCIM_Cubic, leaveTangent=2.2852, tangentMode=RCTM_User, time=1, value=1}
        ]
       }}}
     CurveIndex=$System.ReactivityIntensity
     Scale Curve=12
    )
    Use Spawn Probability=true
   )
  ]
  particleSpawnScript [
   &InitializeParticle(
    Lifetime=&RandomRangeFloat(Minimum=0.35, Maximum=0.45)
    Mesh Scale Mode=::NewEnumerator3
    UsePositionOffset=true
    Write Material Random=true
   )
   /SetVariables(
    Particles.InvertLightning:NiagaraBool=&RandomBool()
    Particles.LightningIndex:NiagaraInt32=&ModuloInt(
     A=&SelectIntFromArray(
      Direct Array Index=&ReturnParticleIDAttributes()
      Direct Array Mode=::NewEnumerator1
      Int Selection Array=$System.Lightning Index
     )
     B=4
    )
    Particles.MirrorLightning:NiagaraBool=&RandomBool()
   ) {isSetParameters=true}
   /SetVariables(Particles.Scale=&Multiply_Vector(
     A=&MakeCustomVectorFromBool(Bool=$Particles.MirrorLightning, True Vector=<-1 1 1>, False Vector=<1 1 1>)
     B=&RandomRangeVector(Minimum=<0.8 0.8 1>, Maximum=<1 1 1>)
    )) {isSetParameters=true}
  ]
  particleUpdateScript [
   &ParticleState()
   &ScaleMeshSize(Scale Factor=&MakeVector(
     X=&Lerp_Float(
      A=1
      B=1.1
      Alpha=&FloatFromCurve(FloatCurve=DataInterface() {propertyValues {Curve {
          defaultValue=0.5
          keys [{interpMode=RCIM_Cubic, tangentMode=RCTM_User, time=0, value=0}, {interpMode=RCIM_Cubic, tangentMode=RCTM_User, time=1, value=1}]
         }}})
     )
     Y=&Lerp_Float(
      A=0.95
      B=1.05
      Alpha=&FloatFromCurve(FloatCurve=DataInterface() {propertyValues {Curve {
          defaultValue=0.5
          keys [{interpMode=RCIM_Cubic, tangentMode=RCTM_User, time=0, value=0}, {interpMode=RCIM_Cubic, tangentMode=RCTM_User, time=1, value=1}]
         }}})
     )
     Z=1
    ))
   &DynamicMaterialParameters(
    Index 0 Param 1=&RandomRangeFloat(Minimum=8, Maximum=10)
    Index 0 Param 2=&Divide_Float(
     A=&RandomRangeFloat(Minimum=1, Maximum=1.5)
     B=&SelectFloatFromArray(Direct Array Index=$Particles.LightningIndex, Float Selection Array=DataInterface() {propertyValues {FloatData [1.5,1.6,1.3,1.5]}})
    )
    Index 0 Param 3=&FloatFromCurve(FloatCurve=DataInterface() {propertyValues {Curve {keys [<at 0 1>,<at 0.1 0.375>,<at 0.5 0.05>,<at 1 0>]}}})
    Index 0 Param 4=&MakeCustomFloatFromBool(Bool=$Particles.InvertLightning, True Float=1, False Float=0)
    Index 1 Param 1=&MakeFloatFromVector(Vector=$Particles.Scale, Channel=::NewEnumerator0)
    Index 1 Param 2=&MakeFloatFromVector(Vector=$Particles.Scale, Channel=::NewEnumerator1)
    Index 1 Param 3=&MakeFloatFromVector(Vector=$Particles.Scale, Channel=::NewEnumerator2)
    Module.Index0Param0Write=true
    Module.Index0Param1Write=true
    Module.Index0Param2Write=true
    Module.Index0Param3Write=true
    Module.Index1Param0Write=true
    Module.Index1Param1Write=true
    Module.Index1Param2Write=true
    Module.Index1Param3Write=true
    Write Parameter Index 1=true
   )
   &Color(
    Color=&ColorFromCurve(ColorCurve=DataInterface() {propertyValues {
       AlphaCurve {keys [<at 0 0>, {arriveTangent=0.3873, interpMode=RCIM_Cubic, leaveTangent=0.3873, time=0.0998, value=1}, <at 0.7746 0.3>, <at 1 0 RCIM_Cubic>]}
       BlueCurve {keys [<at 0 0 RCIM_Cubic>,<at 0.4946 0>,<at 1 0.1 RCIM_Cubic>]}
       GreenCurve {keys [<at 0 5 RCIM_Cubic>,<at 0.4946 0.8333>,<at 1 0.175 RCIM_Cubic>]}
       RedCurve {keys [<at 0 10 RCIM_Cubic>,<at 0.4946 5>,<at 1 1 RCIM_Cubic>]}
      }})
    Scale Alpha=$System.ReactivityIntensity
    Scale Color=<2 2 2>
   )
   &ScaleColor(
    Color Value To Scale=$Particles.Color
    Scale RGB=&MakeVectorFromLinearColorRGB(Color=&FNTimeOfDayColor(Night Color=<rgba 0.56 0.6488 0.8 1>, Morning Color=<rgba 0.6 0.6262 0.8 1>, Day Color=<rgba 1 1 1 1>, Evening Color=<rgba 0.6 0.6 0.6 1>))
    ScaleA=true
    ScaleRGB=true
   )
  ]
  rendererClasses [&NiagaraMeshRendererProperties]
  renderers [{rendererClass=&NiagaraMeshRendererProperties, rendererIndex=0}]
 }
}
summary {
 systemName=NS_Pickaxe_EssayViewMyth_Idle
 emitters [
  {emitterName=Capsule_Glow, rendererClasses [&NiagaraSpriteRendererProperties]}
  {emitterName=Lightning_Splines, rendererClasses [&NiagaraMeshRendererProperties]}
  {emitterName=Glows, rendererClasses [&NiagaraSpriteRendererProperties]}
 ]
 userVariables [
  User.bIsFrontEnd {type=NiagaraBool}
  User.bIsFrontEndPreview {type=NiagaraBool}
  User.DebugReactivity {type=NiagaraBool}
  User.ReactivityOn {type=NiagaraFloat, defaultValue=0, description=""}
  User.ReactivityTime {type=NiagaraFloat, defaultValue=0, description=""}
 ]
}
