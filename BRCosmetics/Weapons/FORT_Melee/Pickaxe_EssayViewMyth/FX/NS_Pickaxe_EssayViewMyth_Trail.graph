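# Niagara swing trail for the EssayViewMyth pickaxe, drawing short-lived gold-orange anim-trail ribbons between two bone or socket indices.
# Hard_and_Soft_Trail spawns two trails, using spawn-group masks to split hard and soft color alphas.
# Lightning_Trail widens over life and scales base and reactive color alphas by System.ReactivityIntensity, so it vanishes at zero.
# Both emitters fade by particle life and an emitter-age window, with time-of-day tint and front-end modifiers.
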NiagaraSystem /BRCosmetics/Weapons/FORT_Melee/Pickaxe_EssayViewMyth/FX/NS_Pickaxe_EssayViewMyth_Trail
dependencies {
 usedDataInterfaces [&NiagaraDataInterfaceColorCurve, &NiagaraDataInterfaceCurve, &NiagaraDataInterfaceSocketReader]
 usedDynamicInputs [
  &MakeCustomFloatFromBool
  &Lerp_Float
  &Divide_Float
  &Subtract_Float
 ]
 usedModules [
  &self:SystemSpawnScript.SystemScriptSource.SystemScriptGraph.NiagaraNodeAssignment_1.SetVariables_#1
  &SystemState
  &self:SystemSpawnScript.SystemScriptSource.SystemScriptGraph.NiagaraNodeAssignment_0.SetVariables_#2
 ]
}
emitters {
 Hard_and_Soft_Trail {
  emitterSpawnScript [&FN_RibbonIDOffsetUtility()]
  emitterUpdateScript [
   &EmitterState()
   &FN_SpawnAnimTrail(AnimTrailWidthMode=::FromSecond, Number of Anim Trails=2)
  ]
  particleSpawnScript [
   &FN_InitializeAnimTrailBoneIndecies(First Bone or Socket Index=1, Second Bone or Socket Index=0)
   &FN_InitializeAnimTrailBoneIndecies(AnimTrail Selection=1, First Bone or Socket Index=1, Second Bone or Socket Index=0) {module=FN_InitializeAnimTrailBoneIndecies001}
   &FN_InitializeAnimTrail(
    AnimTrailWidth=&MaskFloatBySpawnGroup(Masked Float=1, Passthrough Float=0.7)
    Color=<rgba 1 0.5973 0.0024 0.25>
    Lifetime=0.27
   )
  ]
  particleUpdateScript [
   &SolveForcesAndVelocity(Clamp Velocity=false, Limit Acceleration=false)
   &FN_UpdateAnimTrail() {module=FN_UpdateAnimTrail001}
   &ScaleColor(
    Scale Alpha=&FloatFromCurve(
     FloatCurve=DataInterface() {propertyValues {Curve {keys [{interpMode=RCIM_Cubic, tangentMode=RCTM_User, time=0, value=1}, {interpMode=RCIM_Cubic, tangentMode=RCTM_User, time=1, value=0}]}}}
     Scale Curve=&FloatFromCurve(
      FloatCurve=DataInterface() {propertyValues {Curve {keys [
          {arriveTangent=22.4583, interpMode=RCIM_Cubic, leaveTangent=22.4583, tangentMode=RCTM_User, time=0, value=0}
          {arriveTangent=-0.0759, interpMode=RCIM_Cubic, leaveTangent=-0.0759, tangentMode=RCTM_User, time=0.1005, value=1}
          {arriveTangent=-0.3023, interpMode=RCIM_Cubic, leaveTangent=-0.3023, tangentMode=RCTM_User, time=0.2, value=1}
          {arriveTangent=2.5339, interpMode=RCIM_Cubic, leaveTangent=2.5339, tangentMode=RCTM_User, time=0.3, value=0}
         ]}}}
      CurveIndex=$Emitter.Age
      Scale Curve=&Multiply_Float(
       A=&MaskFloatBySpawnGroup(Masked Float=1, Passthrough Float=0)
       B=&FN_FrontEndFloatModifier(Gameplay=&FNTimeOfDayScalar(Night Scale=1, Morning Scale=1, Evening Scale=1), Front End=1)
      )
     )
    )
    Scale RGB=&MakeVectorFromLinearColorRGB(Color=&FNTimeOfDayColor(Night Color=<rgba 0.32 0.4976 0.8 1>, Morning Color=<rgba 0.6 0.6262 0.8 1>, Day Color=<rgba 1 1 1 1>, Evening Color=<rgba 0.6 0.6 0.6 1>))
    ScaleA=true
    ScaleRGB=true
   )
   /SetVariables(Particles.Color_Soft:LinearColor=&MultiplyLinearColor(
     Linear Color=$Particles.Initial.Color
     Scale Factor=&MakeVector4FromVectorAndFloat(
      "Vector (RGB)"=&MakeVectorFromLinearColorRGB(Color=&FNTimeOfDayColor(Night Color=<rgba 0.32 0.4976 0.8 1>, Morning Color=<rgba 0.6 0.6262 0.8 1>, Day Color=<rgba 1 1 1 1>, Evening Color=<rgba 0.6 0.6 0.6 1>))
      "Float (Alpha)"=&FloatFromCurve(
       FloatCurve=DataInterface() {propertyValues {Curve {keys [{interpMode=RCIM_Cubic, tangentMode=RCTM_User, time=0, value=1}, {interpMode=RCIM_Cubic, tangentMode=RCTM_User, time=1, value=0}]}}}
       Scale Curve=&FloatFromCurve(
        FloatCurve=DataInterface() {propertyValues {Curve {keys [
            {arriveTangent=22.4583, interpMode=RCIM_Cubic, leaveTangent=22.4583, tangentMode=RCTM_User, time=0, value=0}
            {arriveTangent=-0.0759, interpMode=RCIM_Cubic, leaveTangent=-0.0759, tangentMode=RCTM_User, time=0.1005, value=1}
            {arriveTangent=-0.3023, interpMode=RCIM_Cubic, leaveTangent=-0.3023, tangentMode=RCTM_User, time=0.2, value=1}
            {arriveTangent=2.5339, interpMode=RCIM_Cubic, leaveTangent=2.5339, tangentMode=RCTM_User, time=0.3, value=0}
           ]}}}
        CurveIndex=$Emitter.Age
        Scale Curve=&Multiply_Float(
         A=&MaskFloatBySpawnGroup(Masked Float=1, Passthrough Float=0, Spawn Group Mask=1)
         B=&FN_FrontEndFloatModifier(Gameplay=&FNTimeOfDayScalar(Night Scale=1, Morning Scale=1, Evening Scale=1), Front End=1)
        )
       )
      )
     )
    )) {isSetParameters=true}
  ]
  rendererClasses [&NiagaraRibbonRendererProperties]
  renderers [
   {rendererClass=&NiagaraRibbonRendererProperties}
   {rendererClass=&NiagaraRibbonRendererProperties, rendererIndex=1}
  ]
 }
 Lightning_Trail {
  emitterSpawnScript [&FN_RibbonIDOffsetUtility()]
  emitterUpdateScript [
   &EmitterState()
   &FN_SpawnAnimTrail(AnimTrailWidthMode=::FromSecond)
  ]
  particleSpawnScript [
   &FN_InitializeAnimTrailBoneIndecies(First Bone or Socket Index=1, Second Bone or Socket Index=0)
   &FN_InitializeAnimTrailBoneIndecies(AnimTrail Selection=1) {module=FN_InitializeAnimTrailBoneIndecies001}
   &FN_InitializeAnimTrail(Color=<rgba 1 0.6667 0 0.3>, Lifetime=0.27)
  ]
  particleUpdateScript [
   &SolveForcesAndVelocity(Clamp Velocity=false, Limit Acceleration=false)
   &FN_UpdateAnimTrail() {module=FN_UpdateAnimTrail001}
   &Color(
    Color=&ColorFromCurve(ColorCurve=DataInterface() {propertyValues {
       AlphaCurve {keys [<at 0 0>, {arriveTangent=0.3873, interpMode=RCIM_Cubic, leaveTangent=0.3873, time=0.0998, value=1}, <at 0.7746 0.3>, <at 1 0 RCIM_Cubic>]}
       BlueCurve {keys [<at 0.0893 0 RCIM_Cubic>,<at 0.6118 0>,<at 1 0 RCIM_Cubic>]}
       GreenCurve {keys [<at 0.0893 0.75 RCIM_Cubic>,<at 0.6118 0.1667>,<at 1 0.0833 RCIM_Cubic>]}
       RedCurve {keys [<at 0.0893 1.5 RCIM_Cubic>,<at 0.6118 1>,<at 1 0.5 RCIM_Cubic>]}
      }})
    Scale Color=<2 2 2>
   )
   &ScaleColor(
    Color Value To Scale=$Particles.Color
    Scale Alpha=&FloatFromCurve(
     FloatCurve=DataInterface() {propertyValues {Curve {keys [{interpMode=RCIM_Cubic, tangentMode=RCTM_User, time=0, value=1}, {interpMode=RCIM_Cubic, tangentMode=RCTM_User, time=1, value=0}]}}}
     Scale Curve=&FloatFromCurve(
      FloatCurve=DataInterface() {propertyValues {Curve {keys [
          {arriveTangent=22.4583, interpMode=RCIM_Cubic, leaveTangent=22.4583, tangentMode=RCTM_User, time=0, value=0}
          {arriveTangent=-0.0759, interpMode=RCIM_Cubic, leaveTangent=-0.0759, tangentMode=RCTM_User, time=0.1005, value=1}
          {arriveTangent=-0.3023, interpMode=RCIM_Cubic, leaveTangent=-0.3023, tangentMode=RCTM_User, time=0.2, value=1}
          {arriveTangent=2.5339, interpMode=RCIM_Cubic, leaveTangent=2.5339, tangentMode=RCTM_User, time=0.3, value=0}
         ]}}}
      CurveIndex=$Emitter.Age
      Scale Curve=&FN_FrontEndFloatModifier(Gameplay=&FNTimeOfDayScalar(Night Scale=1, Morning Scale=1, Evening Scale=1), Front End=1)
     )
    )
    Scale RGB=&MakeVectorFromLinearColorRGB(Color=&FNTimeOfDayColor(Night Color=<rgba 0.32 0.4976 0.8 1>, Morning Color=<rgba 0.6 0.6262 0.8 1>, Day Color=<rgba 1 1 1 1>, Evening Color=<rgba 0.6 0.6 0.6 1>))
    ScaleA=true
    ScaleRGB=true
   )
   /SetVariables(
    Particles.BaseColor:LinearColor=&MultiplyLinearColor(
     Linear Color=$Particles.Color
     Scale Factor=&MakeFloat4FromVectorAndFloat(
      "Vector (RGB)"=&MakeVectorFromLinearColorRGB(Color=&MultiplyLinearColor(Linear Color=<rgba 1 1 1 1>, Scale Factor=<1 1 1 1>))
      "Float (Alpha)"=&FloatFromCurve(
       FloatCurve=DataInterface() {propertyValues {Curve {defaultValue=0, keys [{tangentMode=RCTM_User, time=0, value=0}, {tangentMode=RCTM_User, time=1, value=1}]}}}
       CurveIndex=$System.ReactivityIntensity
      )
     )
    )
    Particles.ReactiveColor:LinearColor=&MultiplyLinearColor(
     Linear Color=$Particles.Color
     Scale Factor=&MakeFloat4FromVectorAndFloat(
      "Vector (RGB)"=<2 2 2>
      "Float (Alpha)"=&FloatFromCurve(
       FloatCurve=DataInterface() {propertyValues {Curve {defaultValue=1, keys [{tangentMode=RCTM_User, time=0, value=0}, {tangentMode=RCTM_User, time=1, value=1}]}}}
       CurveIndex=$System.ReactivityIntensity
      )
     )
    )
   ) {isSetParameters=true}
   &ScaleRibbonWidth(Ribbon Width Scale=&FloatFromCurve(FloatCurve=DataInterface() {propertyValues {Curve {
        defaultValue=1
        keys [
         {arriveTangent=0.829, interpMode=RCIM_Cubic, leaveTangent=0.829, tangentMode=RCTM_User, time=0, value=1}
         {interpMode=RCIM_Cubic, tangentMode=RCTM_User, time=1, value=1.5}
        ]
       }}}))
  ]
  rendererClasses [&NiagaraRibbonRendererProperties]
  renderers [
   {rendererClass=&NiagaraRibbonRendererProperties}
   {rendererClass=&NiagaraRibbonRendererProperties, rendererIndex=1}
  ]
 }
}
summary {
 systemName=NS_Pickaxe_EssayViewMyth_Trail
 emitters [
  {emitterName=Lightning_Trail, rendererClasses [&NiagaraRibbonRendererProperties]}
  {emitterName=Hard_and_Soft_Trail, rendererClasses [&NiagaraRibbonRendererProperties]}
 ]
 userVariables [
  User.bIsFrontEnd {type=NiagaraBool}
  User.bIsFrontEndPreview {type=NiagaraBool}
  User.DebugReactivity {type=NiagaraBool}
  User.ReactivityOn {type=NiagaraFloat, defaultValue=0, description=""}
  User.ReactivityTime {type=NiagaraFloat, defaultValue=0, description=""}
  User.RibbonIDOffset {type=NiagaraInt32}
 ]
}
